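; Burst gameplay cue that configures the blink trail effect for CoffeePuff, apparently a teleport visual on the target pawn.
; On burst it takes the spawned particle component and feeds it source position from the cue location and target position from the actor.
; It derives the pre-telegraph delay from the cue's raw magnitude minus a fixed offset, flags it as not teleporting back in, and sets a portal Z offset.
; The component is then rotated to face opposite the pawn's forward vector.
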(fn OnBurst(EitherVFX-preferthisversion)
  (params (MyTarget :type Actor%20Object%20Reference) (Parameters :type Gameplay%20Cue%20Parameters%20Structure%20%28by%20ref%29) (ParticleComponents :type Array%20of%20FXSystem%20Component%20Object%20References) (AudioComponents :type Array%20of%20Audio%20Component%20Object%20References) (BurstCameraShakeInstance :type Camera%20Shake%20Base%20Object%20Reference) (BurstDecalInstance :type Decal%20Actor%20Object%20Reference))
  (bind (_normalizedmagnitude _rawmagnitude _effectcontext _matchedtagname _originaltag _aggregatedsourcetags _aggregatedtargettags _location) (Ability|GameplayCue|BreakGameplayCueParameters Parameters))
  (bind _output (Utilities|Array|Get(acopy) ParticleComponents))
  (bind _asfort_pawn (Utilities|Casting|CastToFortPawn MyTarget))
  (Effects|Components|ParticleSystem|SetFloatParameter _output "PreTelegraphDelayTime" (- _rawmagnitude 0.3))
  (Effects|Components|ParticleSystem|SetBoolParameter _output "IsTeleportBackIn" false)
  (Effects|Components|ParticleSystem|SetVectorParameter _output "SourcePosition" _location)
  (Effects|Components|ParticleSystem|SetVectorParameter _output "TargetPosition" (Transformation|GetActorLocation MyTarget))
  (Effects|Components|ParticleSystem|SetFloatParameter _output "FirePortalZoffset" 1.2e+02)
  (Transformation|SetWorldRotation _output (Math|Conversions|RotationFromXVector (* (Transformation|GetActorForwardVector _asfort_pawn) "-1.000000,-1.000000,-1.000000")) false false)
  (return))
