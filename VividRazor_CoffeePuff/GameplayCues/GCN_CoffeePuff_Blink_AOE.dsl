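; Gameplay cue that places the burst Niagara effect for the CoffeePuff Blink area-of-effect on the ground at the target player pawn.
; On burst, it line-traces vertically through the pawn's right-hand socket to find the surface.
; It moves a spawned particle component slightly above the hit point and orients it to the surface normal.
; The AOE radius variable drives the effect's FXScale parameter, with no check for a failed cast or trace.
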(fn OnBurst(Niagara)
  (params (MyTarget :type Actor%20Object%20Reference) (Parameters :type Gameplay%20Cue%20Parameters%20Structure%20%28by%20ref%29) (ParticleComponents :type Array%20of%20Niagara%20Particle%20System%20Component%20Object%20References) (AudioComponents :type Array%20of%20Audio%20Component%20Object%20References) (BurstCameraShakeInstance :type Camera%20Shake%20Base%20Object%20Reference) (BurstDecalInstance :type Decal%20Actor%20Object%20Reference))
  (bind _output (Utilities|Array|Get(acopy) ParticleComponents))
  (bind _asfort_player_pawn (Utilities|Casting|CastToFortPlayerPawn MyTarget))
  (bind _outhit (Collision|LineTraceByChannel 0 (+ (Transformation|GetSocketLocation (Variables|/Script/Engine.Character|GetMesh _asfort_player_pawn) "hand_r") "0.000000,0.000000,200.000000") (+ (Transformation|GetSocketLocation (Variables|/Script/Engine.Character|GetMesh _asfort_player_pawn) "hand_r") "0.000000,0.000000,-200.000000") "TraceTypeQuery10"))
  (Transformation|SetWorldLocationAndRotation _output (+ (.Location (Collision|BreakHitResult _outhit)) "0.000000,0.000000,20.000000") (Math|Conversions|RotationFromXVector (.Normal (Collision|BreakHitResult _outhit))) false false)
  (Effects|Components|ParticleSystem|SetFloatParameter _output "FXScale" (LocalVariables|GetFX_AOE_Radius)) ; Radius
  (return))
