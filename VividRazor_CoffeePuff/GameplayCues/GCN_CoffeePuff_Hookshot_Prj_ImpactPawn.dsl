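; Burst gameplay cue that plays visuals and audio when a pawn is hit by the CoffeePuff hookshot projectile.
; OnBurst runs the parent, then moves the first spawned FX component to the target's location and sets its look-at rotation.
; It plays an impact sound there, plus a 2D reel-in whoosh if the target's instigator is locally viewed.
; GetClosestLocationToPawn, a closest-bone lookup falling back to the test location, is not called in these graphs.
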(fn GetClosestLocationToPawn
  (params (FortPawnActor :type Actor%20Object%20Reference) (TestLocation :type Vector))
  ;[0.1155 0.0577 0.15 0.5] Returns the location that the spawned impact niagara effect should move to in terms of world location while remaining attached to the target actor.
  (Utilities|FlowControl|Sequence
    (:then_0
      (bind _asfort_pawn (Utilities|Casting|CastToFortPawn FortPawnActor))
      (bind _mesh (Variables|/Script/Engine.Character|GetMesh _asfort_pawn)
        (:then
          (return (Components|SkinnedMesh|FindClosestBone _mesh TestLocation 0.0 false) _asfort_pawn))
        (:else)))
    (:then_1
      (return TestLocation 0))))

(fn OnBurst(EitherVFX-preferthisversion)
  (params (MyTarget :type Actor%20Object%20Reference) (Parameters :type Gameplay%20Cue%20Parameters%20Structure%20%28by%20ref%29) (ParticleComponents :type Array%20of%20FXSystem%20Component%20Object%20References) (AudioComponents :type Array%20of%20Audio%20Component%20Object%20References) (BurstCameraShakeInstance :type Camera%20Shake%20Base%20Object%20Reference) (BurstDecalInstance :type Decal%20Actor%20Object%20Reference))
  (bind _returnvalue (Transformation|GetActorLocation MyTarget))
  (bind _output (Utilities|Array|Get(acopy) ParticleComponents))
  (bind _instigator (Variables|/Script/Engine.Actor|GetInstigator MyTarget))
  ;[0.02047 0.1342 0.15 0.5] OnBurst: Visuals for when the pawn is hit by a hookshot.
  (|CallParentFunction MyTarget Parameters ParticleComponents AudioComponents BurstCameraShakeInstance BurstDecalInstance)
  ;; Adjust rotation and location of the spawned niagara effect.
  (if (Utilities|Array|IsValidIndex ParticleComponents)
    (Utilities|IsValid _output
      (:Is%20Valid
        (Transformation|SetWorldLocation _output _returnvalue false false)
        (Transformation|SetWorldRotation _output (Math|Rotator|FindLookatRotation (.location (Transformation|GetActorTransform (Variables|GetMyTarget))) (Transformation|GetWorldLocation _output)) false false)
        ;[0.02696 0.05772 0.15 0.5] Audio When Player is Hit by Hookshot
        (Utilities|FlowControl|Sequence
          (:then_0
            ; Impact
            (Audio|PlaySoundatLocation _instigator "/CoffeePuffGameplay/Assets/Sounds/IP_SC_CoffeePuff_Hookshot_Prj_Impact_Player.IP_SC_CoffeePuff_Hookshot_Prj_Impact_Player" _returnvalue))
          (:then_1
            ; Reel-in Woosh on Hit Player
            (if (Pawn|IsLocallyViewed _instigator)
              (Audio|PlaySound2D _instigator "/CoffeePuffGameplay/Assets/Sounds/IP_SC_CoffeePuff_Hookshot_Prj_Damage_Local_Player.IP_SC_CoffeePuff_Hookshot_Prj_Damage_Local_Player")
              (return))))))))
