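; Burst GameplayCue for the CoffeePuff hookshot projectile hitting the world, providing impact visuals and audio.
; OnBurst calls the parent implementation, then sets IsHittingPlayer to false on the first spawned Niagara component to mark a non-player hit.
; It plays the world-impact sound at the impact point taken from the effect context's hit result.
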(fn OnBurst(Niagara)
  (params (MyTarget :type Actor%20Object%20Reference) (Parameters :type Gameplay%20Cue%20Parameters%20Structure%20%28by%20ref%29) (ParticleComponents :type Array%20of%20Niagara%20Particle%20System%20Component%20Object%20References) (AudioComponents :type Array%20of%20Audio%20Component%20Object%20References) (BurstCameraShakeInstance :type Camera%20Shake%20Base%20Object%20Reference) (BurstDecalInstance :type Decal%20Actor%20Object%20Reference))
  (bind (_normalizedmagnitude _rawmagnitude _effectcontext) (Ability|GameplayCue|BreakGameplayCueParameters Parameters))
  (bind (_bblockinghit _binitialoverlap _time _distance _location _impactpoint) (Collision|BreakHitResult (Ability|EffectContext|GetHitResult _effectcontext)))
  (|CallParentFunction MyTarget Parameters ParticleComponents AudioComponents BurstCameraShakeInstance BurstDecalInstance)
  (Effects|Components|ParticleSystem|SetBoolParameter (Utilities|Array|Get(acopy) ParticleComponents) "IsHittingPlayer" false)
  (Audio|PlaySoundatLocation 0 "/CoffeePuffGameplay/Assets/Sounds/IP_SC_CoffeePuff_Hookshot_Prj_Impact_World.IP_SC_CoffeePuff_Hookshot_Prj_Impact_World" _impactpoint)
  (return))
